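# Physics asset for the Spring Break backpack mesh, setting up secondary motion for a compass-like piece (compas) and a hinged piece.
# Simple capsule bodies cover the pack plus a base and dyn bone for each piece, chained pack to base to dyn.
# Base-to-pack joints leave swing and twist free.
# The dyn compas joint allows wide twist with minimal swing, while the dyn hinge joint swings on one axis only.
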PhysicsAsset /BRCosmetics/Accessories/FORT_Backpacks/Backpack_M_MED_SpringBreak/Meshes/M_MED_SpringBreak_Pack_Physics
bodies {
 base_compas [{shapeType=Capsule, length=1, radius=1, shapeName=base_compas_capsule, rotation=<rot 90 0 0>}]
 base_hinge [{shapeType=Capsule, length=1, radius=1, shapeName=base_hinge_capsule, rotation=<rot 90 0 0>}]
 dyn_compas [{shapeType=Capsule, length=1, radius=1, shapeName=dyn_compas_capsule, center=<0 0 -2>, rotation=<rot 90 0 0>}]
 dyn_hinge [{shapeType=Capsule, length=5, radius=5, shapeName=dyn_hinge_capsule, center=<5 0 0>, rotation=<rot 90 0 0>}]
 pack [{shapeType=Capsule, length=1, radius=1, shapeName=pack_capsule, rotation=<rot 90 0 0>}]
}
constraints [
 {
  bone1Name=base_compas
  bone2Name=pack
  swing1LimitDegrees=45
  swing1Motion=Free
  swing2LimitDegrees=45
  swing2Motion=Free
  twistLimitDegrees=45
  twistMotion=Free
 }
 {bone1Name=dyn_compas, bone2Name=base_compas, swing1LimitDegrees=2, swing2LimitDegrees=2, twistLimitDegrees=180}
 {
  bone1Name=base_hinge
  bone2Name=pack
  swing1LimitDegrees=45
  swing1Motion=Free
  swing2LimitDegrees=45
  swing2Motion=Free
  twistLimitDegrees=45
  twistMotion=Free
 }
 {bone1Name=dyn_hinge, bone2Name=base_hinge, swing1LimitDegrees=50, swing2LimitDegrees=0, twistLimitDegrees=0}
]
