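# Anim Blueprint for the Spring Break backpack that adds secondary physics and a randomly spinning compass bone.
# AnimGraph runs RigidBody with overridden gravity, additively rolls base_compas, then offsets the pack bone for male characters.
# Each update refreshes gravity through a deprecated state function and integrates roll by real-time delta, re-randomizing speed on a one-second delay, biased away from zero.
# Gender is resolved on customization load; several tangent-based random-rotation chains remain unconnected to any event.
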AnimBlueprint /BRCosmetics/Accessories/FORT_Backpacks/Backpack_M_MED_SpringBreak/Meshes/M_MED_SpringBreak_Pack_AnimBP
target_skeleton=&/M_MED_SpringBreak_Pack_Skeleton
graphs [
 AnimGraph {nodes [
   n1 = RigidBody!(ComponentPose=Local To Component(LocalPose=Input Pose()), Alpha, OverrideWorldGravity:Vector=Get GravityOverride(), ExternalForce) -> (Pose) {
    bOverrideWorldGravity=true
    overridePhysicsAsset=&/M_MED_SpringBreak_Pack_Physics
    alphaBoolBlend {blendOption=Linear}
   }
   n2 = "Transform (Modify) Bone - Bone: base_compas"(ComponentPose=@n1, Alpha=1, Translation="0,0,0", Rotation=Get Spinning_Rotator() -> (Spinning_Rotator:Rotator), Scale="1,1,1") {
    rotationMode=BMM_Additive
    rotationSpace=BCS_BoneSpace
    translationMode=BMM_Additive
    translationSpace=BCS_BoneSpace
    alphaBoolBlend {blendOption=Linear}
    boneToModify {boneName=base_compas}
   }
   n3 = "Transform (Modify) Bone - Bone: pack"!(ComponentPose=@n2, bAlphaBoolEnabled:bool=Get isMale() -> (isMale:bool), Translation="0,-1.5,0", Rotation="0,0,0", Scale) -> (Pose) {
    alphaInputType=Bool
    rotationMode=BMM_Additive
    rotationSpace=BCS_BoneSpace
    translationMode=BMM_Additive
    translationSpace=BCS_BoneSpace
    alphaBoolBlend {blendOption=Linear}
    boneToModify {boneName=pack}
    translation {y=-1.5}
   }
   Output Pose(Result=Component To Local(ComponentPose=@n3)) {layerGroup=None}
   Get RandRot() -> (RandRot:Rotator)
  ]}
 EventGraph {nodes [
   n1 = Delay(Duration=2)
   n2 = RandomFloat!() -> (ReturnValue:double)
   n3 = float * float(A="Tan (Degrees)"!(A:double=float * float(A=@n2, B=80)) -> (ReturnValue:double), B=-100)
   n4 = Set Sphere_Rotation(execute=[@n5.then, @n1.then], Sphere_Rotation:double=@n3) -> (Output_Get:double)
   > Set Spinning_Rotator(Spinning_Rotator:Rotator=MakeRotator(Roll=@n4)) -> (Output_Get:Rotator)
   > n5 = Delay(Duration=2)
   Delay(Duration=2)
   n6 = GetGameTimeInSeconds()
   n7 = float + float(A=float * float(A=@n6, B=20), B=@n6) {is=title K2Node_CommutativeAssociativeBinaryOperator}
   n8 = GetGameTimeInSeconds()
   n9 = float + float(A=float * float(A=@n8, B=-20), B=@n8)
   n10 = "Tan (Degrees)"(A=RandomFloatInRange(Min=NormalizeAxis!(Angle:float=@n7) -> (ReturnValue:float), Max=NormalizeAxis(Angle=@n9)))
   > n11 = Set Sphere_Rotation(Sphere_Rotation=NormalizeAxis(Angle=float * float(A=@n10, B=-100)))
   > Set Spinning_Rotator(Spinning_Rotator=MakeRotator(Roll=@n11))
   Event BlueprintUpdateAnimation()
   > n12 = Sequence!(execute) -> (then_0, then_1, then_2:Exec, then_3:Exec)
   n13 = Spin FNC!(
    execute=@n12.then_1
    self
    speed:double=Get rand() -> (rand:double)
    InitVal:double=Get!() -> (Spinning_Rotator_Roll:float, Spinning_Rotator_Pitch:float, Spinning_Rotator_Yaw:float)
   ) -> (then, output:double)
   > Set Spinning_Rotator(Spinning_Rotator=MakeRotator(Roll=@n13))
   Set deltaT(execute=@n12.then_2, deltaT:double=GetRealTimeSeconds()) -> (Output_Get:double)
   Delay(execute=@n12.then_3, Duration=1)
   > n14 = Set rand(rand:double=RandomFloatInRange(Min=-100, Max=100)) -> (Output_Get:double)
   > n15 = Branch(Condition="float > float"(A=@n14))
   > Set rand(rand=float + float(A=@n14, B=50) {is=title K2Node_PromotableOperator})
   Set rand(execute=@n15.else, rand=float - float(A=@n14, B=50))
   n16 = Gravity Override State Function DEPRECATED(
    execute=@n12.then_0
    CharacterPartAnimInstance=Self-Reference()
    GameTimeInSeconds=GetGameTimeInSeconds()
    GravityOverrideParams=Get Gravity Override Params()
   )
   > Set GravityOverride(GravityOverride=@n16)
   Event OnCharacterCustomizationLoaded()
   > n17 = Set Owning Pawn(Owning Pawn:Pawn Object Reference=TryGetPawnOwner()) -> (Output_Get:Pawn Object Reference)
   > Is Valid(InputObject=@n17)
   > n18 = Sequence!(execute) -> (then_0, then_1)
   > n19 = Cast To FortPlayerPawn(Object=Get Owning Pawn() -> (Owning Pawn:Pawn Object Reference))
   > n20 = Set Owning Player Pawn(Owning Player Pawn:Fort Player Pawn Object Reference=@n19) -> (Output_Get:Fort Player Pawn Object Reference)
   > Set BodyMesh(BodyMesh=GetSkeletalMeshForPartType(self=@n20))
   n21 = FindGender!(execute=@n18.then_1, Object:Object Reference=Get Owning Pawn()) -> (True:Exec, False:Exec, OutGender:EFortCustomGender Enum=Invalid)
   > Set isMale(isMale:bool="Equal (Enum)"(A:EFortCustomGender Enum=@n21, B:EFortCustomGender Enum=Male)) -> (Output_Get:bool)
   n22 = For Loop()
   RetriggerableDelay(execute=@n22.Completed, Duration=1)
   Do Once(execute=@n22.LoopBody)
   GetGameTimeInSeconds()
   RandomInteger!(Max:int=0) -> (ReturnValue:int=0)
   RandomFloatInRangeFromStream!(Stream:"Random Stream Structure (by ref)", Min:float=0, Max:float=0) -> (ReturnValue:float)
   RandomFloatInRange(Min=0, Max=0)
   GetGameTimeInSeconds()
   n23 = "Tan (Degrees)"(A=float * float(A=0, B=-20))
   RandomFloatInRange(Min=NormalizeAxis(Angle=@n3), Max=NormalizeAxis(Angle=float * float(A=@n23, B=20)))
   TryGetPawnOwner()
  ]}
 Spin FNC {nodes [
   n1 = Spin FNC!() -> (then, speed:double=~, InitVal:double=~) {is=title K2Node_FunctionEntry}
   n2 = float * float(A=@n1, B=float - float(A=GetRealTimeSeconds(), B=Get deltaT() -> (deltaT:double)))
   > Return Node!(execute, output:double=NormalizeAxis(Angle=float + float(A=@n1.InitVal, B=@n2) {is=title K2Node_CommutativeAssociativeBinaryOperator}))
  ]}
]
variables [
 BodyMesh {type=SkeletalMeshComponent}
 deltaT {type=number}
 Gravity Override Params {type=GravityOverrideParamsStruct}
 GravityOverride {type=Vector}
 isMale {type=boolean}
 Owning Pawn {type=Pawn, default=null}
 Owning Player Pawn {type=FortPlayerPawn}
 rand {type=number}
 RandRot {type=Rotator}
 Sphere_Rotation {type=number}
 Spinning_Rotator {type=Rotator}
]
